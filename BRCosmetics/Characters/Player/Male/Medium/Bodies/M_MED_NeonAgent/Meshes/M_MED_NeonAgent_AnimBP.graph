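# Animation Blueprint for the male medium NeonAgent body, adding secondary motion and corrective posing to the copied parent-mesh pose.
# A Control Rig feeds hood-zip AnimDynamics, then weighted constraints and additive bone offsets adjust wires, kneepad, pouch and cheeks through cached poses.
# The head bone shrinks to near zero when the playlist uses custom character parts.
# Each update derives neck, head and knee alphas from bone deltas, some gated by crouch or slide.
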AnimBlueprint /BRCosmetics/Characters/Player/Male/Medium/Bodies/M_MED_NeonAgent/Meshes/M_MED_NeonAgent_AnimBP
target_skeleton=&/M_MED_NeonAgent_Skeleton
graphs [
 AnimGraph {nodes [
   n1 = "Transform (Modify) Bone - Bone: head"!(ComponentPose=Local To Component(LocalPose=Use cached pose 'CRCache_postCorrecctions'()), Alpha=1) -> (Pose) {scaleMode=BMM_Replace, alphaBoolBlend {blendOption=Linear}, boneToModify {boneName=head}, scale=<0.01 0.01 0.01>}
   Output Pose(Result=Blend Poses by bool(
     BlendPose_0=Component To Local(ComponentPose=@n1)
     BlendPose_1=Use cached pose 'CRCache_postCorrecctions'()
     BlendTime_0=0
     BlendTime_1=0
     bActiveValue=Get GetPlayerListUsesCustomCharacterParts() -> (GetPlayerListUsesCustomCharacterParts:bool)
    ) {blendType=Linear}) {layerGroup=None}
   n2 = "Constraint - base_kneepad (PPPPP)"!(
    ComponentPose=Local To Component(LocalPose=Use cached pose 'CRCache_PreScale'())
    Alpha=1
    ConstraintWeights_0=0.4554
    ConstraintWeights_1:float=0.415
    ConstraintWeights_2:float=0.1285
    ConstraintWeights_3:float=0.0009
    ConstraintWeights_4:float=0.0002
   ) -> (Pose) {
    boneToModify {boneName=base_kneepad}
    constraintSetup [
     {targetBone {boneName=deform_knee_out_r}}
     {targetBone {boneName=calf_counter_r}}
     {targetBone {boneName=thigh_counter_r}}
     {targetBone {boneName=thigh_r}}
     {targetBone {boneName=deform_knee_in_r}}
    ]
    constraintWeights [1,1,1,1,1]
   }
   n3 = "Constraint - dyn_ctlDevA (PPPP)"!(
    ComponentPose=@n2
    Alpha=1
    ConstraintWeights_0=0.3645
    ConstraintWeights_1:float=0.3369
    ConstraintWeights_2:float=0.2252
    ConstraintWeights_3:float=0.0735
   ) -> (Pose) {
    boneToModify {boneName=dyn_ctlDevA}
    constraintSetup [{targetBone {boneName=clavicle_l}}, {targetBone {boneName=spine_04}}, {targetBone {boneName=spine_05}}, {targetBone {boneName=spine_03}}]
    constraintWeights [1,1,1,1]
   }
   n4 = "Transform (Modify) Bone - Bone: dyn_comWireB"!(ComponentPose=@n3, Alpha=1, Translation="2.2,0,0") -> (Pose) {translationMode=BMM_Additive, translationSpace=BCS_BoneSpace, boneToModify {boneName=dyn_comWireB}, translation {x=2.2}}
   n5 = "Transform (Modify) Bone - Bone: dyn_comWireB"(ComponentPose=@n4, Alpha=1, Translation="1.670,0.024,0.000") {translationMode=BMM_Additive, translationSpace=BCS_BoneSpace, boneToModify {boneName=dyn_comWireB}, translation=<1.67 0.024>}
   n6 = "Transform (Modify) Bone - Bone: dyn_kneepad"!(ComponentPose=@n5, bAlphaBoolEnabled:bool=True, Rotation="0,-30,6") -> (Pose) {
    alphaInputType=Bool
    rotationMode=BMM_Additive
    translationSpace=BCS_BoneSpace
    boneToModify {boneName=dyn_kneepad}
    rotation {roll=6, yaw=-30}
   }
   n7 = "Transform (Modify) Bone - Bone: dyn_kneepad"!(ComponentPose=@n6, bAlphaBoolEnabled=True, Translation:Vector="0.625,-0.943,-0.431", Rotation="9.078,36.092,32.816") -> (Pose) {
    alphaInputType=Bool
    rotationMode=BMM_Additive
    translationMode=BMM_Additive
    translationSpace=BCS_BoneSpace
    boneToModify {boneName=dyn_kneepad}
    rotation=<rot 9.078 36.092 32.816>
    translation=<0.625 -0.943 -0.431>
   }
   n8 = "Transform (Modify) Bone - Bone: dyn_kneepad"!(ComponentPose=@n7, Alpha:float=1, Translation="1.827,-2.002,-0.751", Rotation="9.078,14.718,32.816") -> (Pose) {
    rotationMode=BMM_Additive
    translationMode=BMM_Additive
    translationSpace=BCS_BoneSpace
    boneToModify {boneName=dyn_kneepad}
    rotation=<rot 9.078 14.718 32.816>
    translation=<1.827 -2.002 -0.751>
   }
   n9 = "Transform (Modify) Bone - Bone: dyn_hipPouch_l"!(ComponentPose=@n8, bAlphaBoolEnabled:bool=True, Translation="0.2,0,0") -> (Pose) {
    alphaInputType=Bool
    translationMode=BMM_Additive
    translationSpace=BCS_BoneSpace
    boneToModify {boneName=dyn_hipPouch_l}
    translation {x=0.2}
   }
   n10 = "Transform (Modify) Bone - Bone: L_cheek_inner"!(ComponentPose=@n9, bAlphaBoolEnabled:bool=True, Translation="0,0.2,0") -> (Pose) {
    alphaInputType=Bool
    translationMode=BMM_Additive
    translationSpace=BCS_BoneSpace
    boneToModify {boneName=L_cheek_inner}
    translation {y=0.2}
   }
   n11 = "Transform (Modify) Bone - Bone: R_cheek_inner"!(ComponentPose=@n10, bAlphaBoolEnabled:bool=True, Translation="0,0.1,0") -> (Pose) {
    alphaInputType=Bool
    translationMode=BMM_Additive
    translationSpace=BCS_BoneSpace
    boneToModify {boneName=R_cheek_inner}
    translation {y=0.1}
   }
   Save cached pose 'CRCache_postCorrecctions'(Pose=Component To Local(ComponentPose=@n11)) {cacheName=CRCache_postCorrecctions}
   n12 = Control Rig(Source=Copy Pose From Mesh() {bUseAttachedParent=true}) {controlRigAssetReference {blueprintRigClass=&/M_MED_NeonAgent_CtrlRig.M_MED_NeonAgent_CtrlRig_C}}
   n13 = Anim Dynamics - Bone: dyn_hoodZip_jnt(ComponentPose=Local To Component(LocalPose=@n12), Alpha=1, ExternalForce="0,0,0") {
    angularDampingOverride=0.99
    bGravityOverrideInSimSpace=true
    bOverrideAngularDamping=true
    bUseGravityOverride=true
    alphaBoolBlend {blendOption=Linear}
    boundBone {boneName=dyn_hoodZip_jnt}
    componentAppliedLinearAccClamp=<300 300 300>
    componentLinearAccScale=<1 1 1>
    componentLinearVelScale=<1 1 1>
    gravityOverride {z=-981}
    physicsBodyDefinitions [{
      sphereCollisionRadius=0
      boundBone {boneName=dyn_hoodZip_jnt}
      boxExtents=<3 3 1>
      constraintSetup {twistAxis=AxisY, angularLimitsMax=<50 0 0>, angularLimitsMin=<-10 0 0>, angularTarget=<0 0 0>}
      localJointOffset=<0 0 -2>
     }]
   }
   n14 = "Constraint - dyn_comWireB (PPPPPP)"!(
    ComponentPose=@n13
    Alpha=1
    ConstraintWeights_0=0.3522
    ConstraintWeights_1:float=0.257
    ConstraintWeights_2:float=0.1345
    ConstraintWeights_3:float=0.109
    ConstraintWeights_4:float=0.0803
    ConstraintWeights_5:float=0.067
   ) -> (Pose) {
    boneToModify {boneName=dyn_comWireB}
    constraintSetup [
     {targetBone {boneName=spine_05}}
     {targetBone {boneName=neck_01}}
     {targetBone {boneName=neck_02}}
     {targetBone {boneName=spine_04}}
     {targetBone {boneName=head}}
     {targetBone {boneName=clavicle_l}}
    ]
    constraintWeights [1,1,1,1,1,1]
   }
   n15 = "Constraint - dyn_comWireA (PPPP)"!(
    ComponentPose=@n14
    Alpha=1
    ConstraintWeights_0=0.4638
    ConstraintWeights_1:float=0.2591
    ConstraintWeights_2:float=0.1899
    ConstraintWeights_3:float=0.0873
   ) -> (Pose) {
    boneToModify {boneName=dyn_comWireA}
    constraintSetup [{targetBone {boneName=head}}, {targetBone {boneName=neck_02}}, {targetBone {boneName=neck_01}}, {targetBone {boneName=spine_05}}]
    constraintWeights [1,1,1,1]
   }
   Save cached pose 'CRCache_PreScale'(Pose=Component To Local(ComponentPose=@n15)) {cacheName=CRCache_PreScale}
  ]}
 EventGraph {nodes [
   Event BlueprintUpdateAnimation()
   n1 = Get BodyMesh() -> (BodyMesh:Skeletal Mesh Component Budgeted Object Reference)
   > Set headRotXPos(headRotXPos:float=float * float(
     A=GetAlphaFromReferenceTransformDelta(SkeletalMeshComponent=@n1, BoneName=neck_01, BaseName=spine_02, DeltaComponentType=EulerX, InputRange=(X=0,Y=15), TwistAxis="1,0,0")
     B:bool=Get bIsCrouching()
    ) {is=title K2Node_PromotableOperator}) -> (Output_Get:float)
   > Set headRotZExt(headRotZExt:float=GetAlphaFromReferenceTransformDelta(SkeletalMeshComponent=@n1, BoneName=head, BaseName=spine_05, InputRange=(X=-60,Y=-80), TwistAxis="0,0,1")) -> (Output_Get:float)
   n2 = Get bIsSliding()
   > Set LeftKneeZrot(LeftKneeZrot:float=float * float(
     A=GetAlphaFromReferenceTransformDelta(SkeletalMeshComponent=@n1, BoneName=thigh_r, BaseName=calf_r, InputRange=(X=-110,Y=-132), TwistAxis="1,0,0")
     B=NOT Boolean(A=@n2)
    )) -> (Output_Get:float)
   Event BlueprintInitializeAnimation()
   > Set GetPlayerListUsesCustomCharacterParts(GetPlayerListUsesCustomCharacterParts:bool=GetPlaylistUsesCustomCharacterParts!() -> (ReturnValue:bool)) -> (Output_Get:bool)
   n3 = Event OnCharacterCustomizationLoaded()
   > Set BodyMesh(BodyMesh:Skeletal Mesh Component Budgeted Object Reference=Get Array Item with EFort Custom Part Type(Array=@n3.CustomMeshParts)) -> (Output_Get:Skeletal Mesh Component Budgeted Object Reference)
  ]}
]
variables [
 BodyMesh {type=SkeletalMeshComponentBudgeted}
 GetPlayerListUsesCustomCharacterParts {type=boolean}
 headRotXPos {type=number}
 headRotZExt {type=number}
 LeftKneeZrot {type=number}
 TestVar {type=boolean}
]
